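# Cosmetic AnimBP for the InflatoDodo body that copies the parent mesh pose and adds secondary motion to collar and vest.
# Collar AnimDynamics feeds a Control Rig, cached for two vest branches.
# Front end uses only vest dynamics; otherwise, looser-limited dynamics add bone offsets and look-ats toward the thighs.
# The update event toggles "Hide Element 01" for backbling and computes gravity and thigh-raise values, which the AnimGraph never visibly reads.
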AnimBlueprint /BRCosmetics/Characters/Player/Male/Medium/Bodies/M_MED_InflatoDodo/Meshes/M_MED_InflatoDodo_AnimBP
target_skeleton=&/M_MED_InflatoDodo_Skeleton
graphs [
 AnimGraph {nodes [
   n1 = Anim Dynamics - Bone: dyn_vest_01_r!(ComponentPose=Local To Component(LocalPose=Use cached pose 'Cache'()), bAlphaBoolEnabled:bool=True, ExternalForce="0,0,0") -> (Pose) {
    alphaInputType=Bool
    angularDampingOverride=0.999
    bOverrideAngularDamping=true
    simulationSpace=BoneRelative
    alphaBoolBlend {blendInTime=0.3, blendOption=ExpInOut, blendOutTime=0.3}
    boundBone {boneName=dyn_vest_01_r}
    componentAppliedLinearAccClamp=<2000 2000 2000>
    componentLinearAccScale=<0.1 0.1 0.1>
    componentLinearVelScale=<0.1 0.1 0.1>
    gravityOverride {z=800}
    physicsBodyDefinitions [{
      boundBone {boneName=dyn_vest_01_r}
      boxExtents=<20 20 20>
      constraintSetup {angularLimitsMax=<10 50 2.5>, angularLimitsMin=<-10 -5 -2.5>, angularTarget=<0 0 0>}
      localJointOffset=<4 0 0>
     }]
    relativeSpaceBone {boneName=root}
   }
   n2 = Anim Dynamics - Bone: dyn_vest_01_l!(ComponentPose=@n1, bAlphaBoolEnabled:bool=True, ExternalForce="0,0,0") -> (Pose) {
    alphaInputType=Bool
    angularDampingOverride=0.999
    bOverrideAngularDamping=true
    simulationSpace=BoneRelative
    alphaBoolBlend {blendInTime=0.3, blendOption=ExpInOut, blendOutTime=0.3}
    boundBone {boneName=dyn_vest_01_l}
    componentAppliedLinearAccClamp=<2000 2000 2000>
    componentLinearAccScale=<0.1 0.1 0.1>
    componentLinearVelScale=<0.1 0.1 0.1>
    physicsBodyDefinitions [{
      boundBone {boneName=dyn_vest_01_l}
      boxExtents=<20 20 20>
      constraintSetup {angularLimitsMax=<3 0 2.5>, angularLimitsMin=<-3 -50 -2.5>, angularTarget=<0 0 0>}
      localJointOffset=<4 0 0>
     }]
    relativeSpaceBone {boneName=root}
   }
   n3 = "Transform (Modify) Bone - Bone: dyn_vest_01_l_MSTR"(ComponentPose=@n2, Alpha=1, Translation="-3,-1,0", Rotation="79.997,-0.221,-20.206", Scale="1,1,1") {
    rotationMode=BMM_Additive
    rotationSpace=BCS_BoneSpace
    translationMode=BMM_Additive
    translationSpace=BCS_BoneSpace
    alphaBoolBlend {blendInTime=0.5, blendOption=ExpInOut, blendOutTime=0.5}
    alphaScaleBiasClamp {interpSpeedDecreasing=6}
    boneToModify {boneName=dyn_vest_01_l_MSTR}
    rotation=<rot 79.9971 -0.221 -20.206>
    translation=<-3 -1>
   }
   n4 = Look At - Bone: dyn_vest_01_l_MSTR(ComponentPose=@n3, Alpha=1) {
    interpolationType=EaseInOutExponent2
    alphaBoolBlend {blendInTime=0.5, blendOption=ExpInOut, blendOutTime=0.5}
    boneToModify {boneName=dyn_vest_01_l_MSTR}
    lookAtLocation=<-31.4249 6.042 0.6191>
    lookAtTarget {boneReference {boneName=thigh_l}}
    lookAt_Axis {axis=<1 0>}
   }
   n5 = "Transform (Modify) Bone - Bone: dyn_vest_01_r_MSTR"(ComponentPose=@n4, Alpha=1, Translation="-3,-1,0", Rotation="-80,0,-20", Scale="1,1,1") {
    rotationMode=BMM_Additive
    rotationSpace=BCS_BoneSpace
    translationMode=BMM_Additive
    translationSpace=BCS_BoneSpace
    alphaBoolBlend {blendInTime=0.5, blendOption=ExpInOut, blendOutTime=0.5}
    alphaScaleBiasClamp {interpSpeedDecreasing=6}
    boneToModify {boneName=dyn_vest_01_r_MSTR}
    rotation {pitch=-80, roll=-20}
    translation=<-3 -1>
   }
   n6 = Look At - Bone: dyn_vest_01_r_MSTR(ComponentPose=@n5, Alpha=1) {
    interpolationType=EaseInOutExponent2
    alphaBoolBlend {blendInTime=1, blendOutTime=1}
    boneToModify {boneName=dyn_vest_01_r_MSTR}
    lookAtLocation=<9.9422 -0.0345 8.5548>
    lookAtTarget {boneReference {boneName=thigh_r}}
    lookAt_Axis {axis=<1 0>}
   }
   n7 = Anim Dynamics - Bone: dyn_vest_01_r(ComponentPose=Local To Component(LocalPose=Use cached pose 'Cache'()), bAlphaBoolEnabled=True, ExternalForce="0,0,0") {
    alphaInputType=Bool
    angularDampingOverride=0.9999
    bOverrideAngularDamping=true
    simulationSpace=BoneRelative
    alphaBoolBlend {blendInTime=0.3, blendOption=ExpInOut, blendOutTime=0.3}
    boundBone {boneName=dyn_vest_01_r}
    componentAppliedLinearAccClamp=<2000 2000 2000>
    componentLinearAccScale=<0.1 0.1 0.1>
    componentLinearVelScale=<0.1 0.1 0.1>
    gravityOverride {z=800}
    physicsBodyDefinitions [{
      boundBone {boneName=dyn_vest_01_r}
      boxExtents=<20 20 20>
      constraintSetup {angularLimitsMax=<5 20 2.5>, angularLimitsMin=<-5 -5 -2.5>, angularTarget=<0 0 0>}
      localJointOffset=<10 0 0>
     }]
    relativeSpaceBone {boneName=root}
   }
   n8 = Anim Dynamics - Bone: dyn_vest_01_l(ComponentPose=@n7, bAlphaBoolEnabled=True, ExternalForce="0,0,0") {
    alphaInputType=Bool
    angularDampingOverride=0.9999
    bOverrideAngularDamping=true
    simulationSpace=BoneRelative
    alphaBoolBlend {blendInTime=0.3, blendOption=ExpInOut, blendOutTime=0.3}
    boundBone {boneName=dyn_vest_01_l}
    componentAppliedLinearAccClamp=<2000 2000 2000>
    componentLinearAccScale=<0.1 0.1 0.1>
    componentLinearVelScale=<0.1 0.1 0.1>
    physicsBodyDefinitions [{
      boundBone {boneName=dyn_vest_01_l}
      boxExtents=<20 20 20>
      constraintSetup {angularLimitsMax=<5 5 2.5>, angularLimitsMin=<-5 -20 -2.5>, angularTarget=<0 0 0>}
      localJointOffset=<10 0 0>
     }]
    relativeSpaceBone {boneName=root}
   }
   Output Pose(Result=Blend Poses by bool(BlendPose_0=Component To Local(ComponentPose=@n8), BlendPose_1=Component To Local(ComponentPose=@n6), bActiveValue=Get bIsInFrontEnd()) {blendType=Linear}) {layerGroup=None}
   n9 = Anim Dynamics - Bone: dyn_collar_01_r(ComponentPose=Local To Component(LocalPose=Copy Pose From Mesh() {bUseAttachedParent=true}), Alpha=1, ExternalForce="0,0,0") {
    angularDampingOverride=0.9999
    bOverrideAngularDamping=true
    bUseGravityOverride=true
    simulationSpace=BoneRelative
    alphaBoolBlend {blendOption=Linear}
    boundBone {boneName=dyn_collar_01_r}
    componentAppliedLinearAccClamp=<2000 2000 2000>
    componentLinearAccScale=<0.1 0.1 0.1>
    componentLinearVelScale=<0.1 0.1 0.1>
    gravityOverride {z=980}
    physicsBodyDefinitions [{
      boundBone {boneName=dyn_collar_01_r}
      boxExtents=<25 25 25>
      constraintSetup {angularLimitsMax=<5 5 10>, angularLimitsMin=<-5 -5 -10>, angularTarget=<0 0 0>}
      localJointOffset=<10 0 0>
     }]
    relativeSpaceBone {boneName=pelvis}
   }
   n10 = Anim Dynamics - Bone: dyn_collar_01_l(ComponentPose=@n9, Alpha=1, ExternalForce="0,0,0") {
    angularDampingOverride=0.9999
    bOverrideAngularDamping=true
    bUseGravityOverride=true
    simulationSpace=BoneRelative
    alphaBoolBlend {blendOption=Linear}
    boundBone {boneName=dyn_collar_01_l}
    componentAppliedLinearAccClamp=<2000 2000 2000>
    componentLinearAccScale=<0.1 0.1 0.1>
    componentLinearVelScale=<0.1 0.1 0.1>
    gravityOverride {z=980}
    physicsBodyDefinitions [{
      boundBone {boneName=dyn_collar_01_l}
      boxExtents=<25 25 25>
      constraintSetup {angularLimitsMax=<5 5 10>, angularLimitsMin=<-5 -5 -10>, angularTarget=<0 0 0>}
      localJointOffset=<10 0 0>
     }]
    relativeSpaceBone {boneName=pelvis}
   }
   n11 = Control Rig(Source=Component To Local(ComponentPose=@n10)) {
    alphaBoolBlend {blendOption=Linear}
    controlRigAssetReference {blueprintRigClass=&/M_MED_InflatoDodo_CtrlRig.M_MED_InflatoDodo_CtrlRig_C}
   }
   Save cached pose 'Cache'(Pose=@n11) {cacheName=Cache}
  ]}
 EventGraph {nodes [
   n1 = Event BlueprintUpdateAnimation()
   > Set Delta Time X(Delta Time X=@n1.DeltaTimeX)
   > n2 = Sequence!(execute) -> (then_0, then_1, then_2:Exec, then_3:Exec, then_4:Exec)
   n3 = GetDeltaTransformFromRefPose(self=GetOwnerSkeletalMeshComponent(), BoneName=pelvis, BaseName=thigh_r)
   Set ThighUp_R(
    execute=@n2.then_1
    ThighUp_R:double=MapRangeClamped(
     Value=BreakTransform!(InTransform=@n3) -> (Location, Rotation_Roll:float, Rotation_Pitch:float, Rotation_Yaw:float, Scale).Rotation_Yaw
     InRangeA=50
     InRangeB=90
    )
   ) -> (Output_Get:double)
   n4 = GetDeltaTransformFromRefPose(self=GetOwnerSkeletalMeshComponent(), BoneName=pelvis, BaseName=thigh_l)
   > Set Thigh Up L(Thigh Up L:double=MapRangeClamped(Value=BreakTransform(InTransform=@n4).Rotation_Yaw, InRangeA=50, InRangeB=90)) -> (Output_Get:double)
   n5 = Get Dynamic Material Instance(execute=@n2.then_2, SkeletalMeshComponent=GetPartSkeletalMeshComponent())
   > n6 = Branch(Condition=CheckForBackbling())
   > SetScalarParameterValue(self=@n5, ParameterName=Hide Element 01, Value=1)
   SetScalarParameterValue(execute=@n6.else, self=@n5, ParameterName=Hide Element 01, Value=0)
   n7 = Gravity Override State Function DEPRECATED(
    execute=@n2.then_0
    CharacterPartAnimInstance=Self-Reference()
    GameTimeInSeconds=GetGameTimeInSeconds()
    GravityOverrideParams=Get Gravity Override Params()
    RBAN Params=Get RBAN Params()
   )
   n8 = Get bIsJumping()
   n9 = Get bIsFalling()
   > Set Gravity(Gravity:Vector=SelectVector(A="0,0,100", B=@n7, bPickA=OR Boolean(A=@n8, B=@n9))) -> (Output_Get:Vector)
   Event OnCharacterAndPickaxeCustomizationLoaded!() -> (
    OutputDelegate
    then
    CustomBaseMesh:Skeletal Mesh Component Object Reference
    CustomParts:Array of Custom Character Part Object References
    CustomMeshParts:Array of Skeletal Mesh Component Budgeted Object References
   ) {is=title K2Node_Event}
  ]}
]
variables [
 Delta Time X {type=number}
 Gravity {type=Vector}
 Gravity Override Params {type=GravityOverrideParamsStruct}
 RBAN Params {
  type=RBANParamsStruct
  default {
   dBNO {simSpaceSettings {
     maxAngularAcceleration=1000
     maxAngularVelocity=3
     maxLinearAcceleration=300
     maxLinearVelocity=700
     velocityScaleZ=0
     externalLinearDragV=<0 0.5 0>
    }}
   dBNO_Carried {simSpaceSettings {
     maxAngularAcceleration=1000
     maxAngularVelocity=3
     maxLinearAcceleration=300
     maxLinearVelocity=700
     velocityScaleZ=0
     externalLinearDragV=<0 0.5 0>
    }}
   emoteOrMelee {simSpaceSettings {
     maxAngularAcceleration=1000
     maxAngularVelocity=3
     maxLinearAcceleration=300
     maxLinearVelocity=700
     velocityScaleZ=0
     externalLinearDragV=<0 0.5 0>
    }}
   front_End {simSpaceSettings {
     maxAngularAcceleration=1000
     maxAngularVelocity=3
     maxLinearAcceleration=300
     maxLinearVelocity=700
     velocityScaleZ=0
     externalLinearDragV=<0 0.5 0>
    }}
   onGround_Crouch_Moving {simSpaceSettings {
     maxAngularAcceleration=1000
     maxAngularVelocity=3
     maxLinearAcceleration=300
     maxLinearVelocity=700
     velocityScaleZ=0
     externalLinearDragV=<0 0.5 0>
    }}
   onGround_Falling {simSpaceSettings {
     maxAngularAcceleration=1000
     maxAngularVelocity=3
     maxLinearAcceleration=300
     maxLinearVelocity=700
     velocityScaleZ=0
     externalLinearDragV=<0 0.5 0>
    }}
   onGround_Moving {simSpaceSettings {
     maxAngularAcceleration=1000
     maxAngularVelocity=3
     maxLinearAcceleration=300
     maxLinearVelocity=700
     velocityScaleZ=0
     externalLinearDragV=<0 0.5 0>
    }}
   onGround_Standing {simSpaceSettings {
     maxAngularAcceleration=1000
     maxAngularVelocity=3
     maxLinearAcceleration=300
     maxLinearVelocity=700
     velocityScaleZ=0
     externalLinearDragV=<0 0.5 0>
    }}
   rail_Grinding {simSpaceSettings {
     maxAngularAcceleration=1000
     maxAngularVelocity=3
     maxLinearAcceleration=300
     maxLinearVelocity=700
     velocityScaleZ=0
     worldAlpha=0.1
     externalLinearDragV=<0 0.5 0>
    }}
   riding_Moving {simSpaceSettings {
     maxAngularAcceleration=1000
     maxAngularVelocity=3
     maxLinearAcceleration=300
     maxLinearVelocity=700
     velocityScaleZ=0
     externalLinearDragV=<0 0.5 0>
    }}
   riding_Moving_Targeting {simSpaceSettings {
     maxAngularAcceleration=1000
     maxAngularVelocity=3
     maxLinearAcceleration=300
     maxLinearVelocity=700
     velocityScaleZ=0
     externalLinearDragV=<0 0.5 0>
    }}
   riding_Standing {simSpaceSettings {
     maxAngularAcceleration=1000
     maxAngularVelocity=3
     maxLinearAcceleration=300
     maxLinearVelocity=700
     velocityScaleZ=0
     externalLinearDragV=<0 0.5 0>
    }}
   skydive_DivingDown {simSpaceSettings {
     maxAngularAcceleration=1000
     maxAngularVelocity=3
     maxLinearAcceleration=300
     maxLinearVelocity=700
     velocityScaleZ=0
     externalLinearDragV=<0 0.5 0>
    }}
   skydive_FloatingDown {
    componentLinearAccScale=<0 0.5 0>
    componentLinearVelScale=<0 0.5 0>
    simSpaceSettings {
     maxAngularAcceleration=1000
     maxAngularVelocity=3
     maxLinearAcceleration=300
     maxLinearVelocity=700
     velocityScaleZ=0
     externalLinearDragV=<0 0.5 0>
    }
   }
   skydive_GoingUp {simSpaceSettings {
     maxAngularAcceleration=1000
     maxAngularVelocity=3
     maxLinearAcceleration=300
     maxLinearVelocity=700
     velocityScaleZ=0
     externalLinearDragV=<0 0.5 0>
    }}
   skydive_Parachuting {
    componentLinearAccScale=<0 0.1 0.3>
    componentLinearVelScale=<0 0.1 0.3>
    simSpaceSettings {
     maxAngularAcceleration=1000
     maxAngularVelocity=3
     maxLinearAcceleration=300
     maxLinearVelocity=700
     velocityScaleZ=0
     externalLinearDragV=<0 0.5 0>
    }
   }
   skydive_Surfing {simSpaceSettings {
     maxAngularAcceleration=1000
     maxAngularVelocity=3
     maxLinearAcceleration=300
     maxLinearVelocity=700
     velocityScaleZ=0
     externalLinearDragV=<0 0.5 0>
    }}
   swimming {simSpaceSettings {
     maxAngularAcceleration=1000
     maxAngularVelocity=3
     maxLinearAcceleration=300
     maxLinearVelocity=700
     velocityScaleZ=0
     externalLinearDragV=<0 0.5 0>
    }}
   tactical_Sprint {simSpaceSettings {
     maxAngularAcceleration=1000
     maxAngularVelocity=3
     maxLinearAcceleration=300
     maxLinearVelocity=700
     velocityScaleZ=0
     externalLinearDragV=<0 0.5 0>
    }}
   ziplining {simSpaceSettings {
     maxAngularAcceleration=1000
     maxAngularVelocity=3
     maxLinearAcceleration=300
     maxLinearVelocity=700
     velocityScaleZ=0
     externalLinearDragV=<0 0.5 0>
    }}
  }
 }
 Thigh Up L {type=number}
 ThighUp_R {type=number}
]
