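# Animation Blueprint for the InflatoDodo face accessory part, following the body pose and curves copied from the attached parent mesh.
# Curves jaw_open_pose and jaw_up_pose drive additive rotations on the jaw bone.
# A weighted multi-bone constraint places the neck spike master, weighted mostly to head and upper neck.
# A Control Rig runs last, fed a HeadUp alpha mapped each update from head yaw relative to spine_05's reference pose.
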AnimBlueprint /BRCosmetics/Characters/Player/Male/Medium/Bodies/M_MED_InflatoDodo/Meshes/Parts/M_MED_InflatoDodo_FaceAcc_AnimBP
target_skeleton=&/M_MED_InflatoDodo_FaceAcc_Skeleton
graphs [
 AnimGraph {nodes [
   n1 = "Transform (Modify) Bone - Bone: deform_Jaw"!(
    ComponentPose=Local To Component(LocalPose=Copy Pose From Mesh() {bCopyCurves=true, bUseAttachedParent=true})
    AlphaCurveName:Name=jaw_open_pose
    Translation="0,0,0"
    Rotation="-18,0,0"
    Scale="1,1,1"
   ) -> (Pose) {
    alphaInputType=Curve
    rotationMode=BMM_Additive
    rotationSpace=BCS_BoneSpace
    alphaBoolBlend {blendOption=Linear}
    boneToModify {boneName=deform_Jaw}
    rotation {pitch=-18}
   }
   n2 = "Transform (Modify) Bone - Bone: deform_Jaw"(ComponentPose=@n1, AlphaCurveName=jaw_up_pose, Translation="0,0,0", Rotation="10,0,0", Scale="1,1,1") {
    alphaInputType=Curve
    rotationMode=BMM_Additive
    rotationSpace=BCS_BoneSpace
    alphaBoolBlend {blendOption=Linear}
    boneToModify {boneName=deform_Jaw}
    rotation {pitch=10}
   }
   n3 = "Constraint - deform_NeckSpike_MSTR (PPPPPPPP)"!(
    ComponentPose=@n2
    Alpha=1
    ConstraintWeights_0=0.0001
    ConstraintWeights_1:float=0.0002
    ConstraintWeights_2:float=0
    ConstraintWeights_3:float=0.0001
    ConstraintWeights_4:float=0.736
    ConstraintWeights_5:float=0.1958
    ConstraintWeights_6:float=0.0368
    ConstraintWeights_7:float=0.031
   ) -> (Pose) {
    alphaBoolBlend {blendOption=Linear}
    boneToModify {boneName=deform_NeckSpike_MSTR}
    constraintSetup [
     {targetBone {boneName=clavicle_l}}
     {targetBone {boneName=clavicle_r}}
     {targetBone {boneName=deform_pec_aim_base_r}}
     {targetBone {boneName=deform_pec_aim_base_l}}
     {targetBone {boneName=head}}
     {targetBone {boneName=neck_02}}
     {targetBone {boneName=neck_01}}
     {targetBone {boneName=spine_05}}
    ]
    constraintWeights [1,1,1,1,1,1,1,1]
   }
   n4 = Control Rig!(Source=Component To Local(ComponentPose=@n3), Alpha, HeadupAlpha:double=Get HeadUp_Alpha() -> (HeadUp_Alpha:double)) -> (Pose) {
    alphaBoolBlend {blendOption=Linear}
    controlRigAssetReference {blueprintRigClass=&/M_MED_InflatoDodo_FaceAcc_CtrlRig.M_MED_InflatoDodo_FaceAcc_CtrlRig_C}
   }
   Output Pose(Result=@n4) {layerGroup=None}
  ]}
 EventGraph {nodes [
   Event BlueprintUpdateAnimation()
   n1 = GetDeltaTransformFromRefPose(self=GetOwnerSkeletalMeshComponent(), BoneName=head, BaseName=spine_05)
   > Set HeadUp_Alpha(HeadUp_Alpha:double=MapRangeClamped(
     Value=BreakTransform!(InTransform=@n1) -> (Location, Rotation_Roll:float, Rotation_Pitch:float, Rotation_Yaw:float, Scale).Rotation_Yaw
     InRangeA=0
     InRangeB=-15
    )) -> (Output_Get:double)
  ]}
]
variables [
 Body Mesh {type=SkeletalMeshComponent}
 HeadUp_Alpha {type=number}
 JawOpen_Alpha {type=number}
 JawOpenPose {type=number}
 JawUp_Alpha {type=number}
 JawUpPose {type=number}
]
